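# Control Rig for the MistMylar ClawArmor medium male body that runs the shared DeformRig function, then outfit-specific corrective passes.
# Clavicle and foot rotations drive armor straps and shoe backs; spherical pose readers on right upper arm and thighs drive deform and rear cloth bones.
# Weighted parent constraints pin sash root and thigh zippers to body bones.
# Front sashes aim at targets offset by thigh-up animation attributes, with offsets smoothly faded out while IsMoving.
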ControlRigBlueprint /BRCosmetics/Characters/Player/Male/Medium/Bodies/M_MED_MistMylar_ClawArmor/Meshes/M_MED_MistMylar_ClawArmor_CtrlRig
graphs [
 New Function_ContainedGraph {nodes [
   RigVMFunctionEntryNode()
   > RigVMFunction_Sequence = RigVMAggregateNode()
   Get Delta Ref = RigVMFunctionReferenceNode(Item=(Type=Bone,Name="clavicle_l"), ExecuteContext=@RigVMFunction_Sequence.A)
   Remap = &RigVMFunction_MathDoubleRemap(SourceMaximum=50, SourceMinimum=0, TargetMaximum=1, TargetMinimum=0, Value=@"Get Delta Ref".Result.Rotation.Pitch)
   > &RigUnit_ModifyTransforms(
    ItemToModify=((Item=(Type=Bone,Name="deform_armor_strap_base_l"),Transform=(Rotation=(X=0.0503,Y=-0.4992,Z=-0.0867,W=0.8607),Translation=(X=-4.038,Y=0,Z=3.9067),Scale3D=(X=1,Y=1,Z=1))))
    WeightMinimum=0
    Weight=@Remap.Result
   )
   Get Delta Ref_1 = RigVMFunctionReferenceNode(Item=(Type=Bone,Name="clavicle_r"), ExecuteContext=@RigVMFunction_Sequence.B)
   > &RigUnit_ModifyTransforms(
    ItemToModify=((Item=(Type=Bone,Name="deform_armor_strap_base_r"),Transform=(Rotation=(X=0.0503,Y=-0.4992,Z=-0.0867,W=0.8607),Translation=(X=4.038,Y=0,Z=-3.9067),Scale3D=(X=1,Y=1,Z=1))))
    WeightMinimum=0
    Weight=&RigVMFunction_MathDoubleRemap(SourceMaximum=50, SourceMinimum=0, TargetMaximum=1, TargetMinimum=0, Value=@"Get Delta Ref_1".Result.Rotation.Pitch).Result
   )
   RigVMFunctionReturnNode(ExecuteContext=@RigVMFunction_Sequence.C)
  ]}
 New Function_ContainedGraph {nodes [
   RigVMFunctionEntryNode()
   > RigVMFunction_Sequence = RigVMAggregateNode()
   Get Delta Ref = RigVMFunctionReferenceNode(Item=(Type=Bone,Name="foot_l"), ExecuteContext=@RigVMFunction_Sequence.A)
   Remap = &RigVMFunction_MathDoubleRemap(SourceMaximum=60, SourceMinimum=0, TargetMaximum=1, TargetMinimum=0, Value=@"Get Delta Ref".Result.Rotation.Yaw)
   > &RigUnit_ModifyTransforms(
    ItemToModify=((Item=(Type=Bone,Name="deform_shoe_rear_base_l"),Transform=(Rotation=(X=0,Y=0,Z=-0.5058,W=0.8627),Translation=(X=-0.5128,Y=-1.0911,Z=0),Scale3D=(X=1,Y=1,Z=1))))
    WeightMinimum=0
    Weight=@Remap.Result
   )
   Get Delta Ref_1 = RigVMFunctionReferenceNode(Item=(Type=Bone,Name="foot_r"), ExecuteContext=@RigVMFunction_Sequence.B)
   > &RigUnit_ModifyTransforms(
    ItemToModify=((Item=(Type=Bone,Name="deform_shoe_rear_base_r"),Transform=(Rotation=(X=0,Y=0,Z=-0.5058,W=0.8627),Translation=(X=0.5128,Y=1.0911,Z=0),Scale3D=(X=1,Y=1,Z=1))))
    WeightMinimum=0
    Weight=&RigVMFunction_MathDoubleRemap(SourceMaximum=60, SourceMinimum=0, TargetMaximum=1, TargetMinimum=0, Value=@"Get Delta Ref_1".Result.Rotation.Yaw).Result
   )
   RigVMFunctionReturnNode(ExecuteContext=@RigVMFunction_Sequence.C)
  ]}
 New Function_ContainedGraph {nodes [
   RigVMFunctionEntryNode()
   > RigVMFunction_Sequence = RigVMAggregateNode()
   &RigUnit_ParentConstraint(
    AdvancedSettings=(RotationOrderForFilter=XZY)
    Child=(Type=Bone,Name="dyn_sash_root")
    Parents=((Item=(Type=Bone,Name="spine_04"),Weight=0.0145),(Item=(Type=Bone,Name="spine_03"),Weight=0.1169),(Item=(Type=Bone,Name="spine_01"),Weight=0.1623),(Item=(Type=Bone,Name="spine_02"),Weight=0.1934),(Item=(Type=Bone,Name="pelvis"),Weight=0.5129))
    Weight=1
    ExecutePin=@RigVMFunction_Sequence.A
   )
   n1 = &RigUnit_SphericalPoseReader(
    ActiveRegionScaleFactors=(PositiveWidth=3,NegativeWidth=0,PositiveHeight=3,NegativeHeight=3)
    ActiveRegionSize=0.1
    Debug=(bDrawDebug=True,bDraw2D=False,bDrawLocalAxes=False,DebugScale=25,DebugSegments=20,DebugThickness=0.25)
    DriverAxis=(X=-1,Y=0,Z=0)
    DriverItem=(Type=Bone,Name="thigh_l")
    FalloffRegionScaleFactors=(PositiveWidth=1,NegativeWidth=1,PositiveHeight=1,NegativeHeight=1)
    FalloffSize=0.3
    FlipHeightScaling=False
    FlipWidthScaling=False
    OptionalParentItem=(Type=Bone,Name="None")
    RotationOffset=(X=-90,Y=0,Z=-30)
    ExecutePin=@RigVMFunction_Sequence.B
   )
   > &RigUnit_ModifyTransforms(
    ItemToModify="((Item=(Type=Bone,Name=\"dyn_rear_cloth_l_MSTR\"),Transform=(Rotation=(X=0.0818,Y=0.1241,Z=-0.5284,W=0.8359),Translation=(X=-1.2013,Y=-0.1044,Z=-1.868),Scale3D=(X=1,Y=1,Z=1))),(Item=(Type=Bone,Name=\"dyn_rear_cloth_c_MSTR\"),Transform=(Rotation=(X=0,Y=0,Z=-0.4 ...(truncated)"
    WeightMinimum=0
    Weight=@n1.OutputParam
   )
   > n2 = &RigUnit_SphericalPoseReader(
    ActiveRegionScaleFactors=(PositiveWidth=0,NegativeWidth=3,PositiveHeight=3,NegativeHeight=3)
    ActiveRegionSize=0.1
    Debug=(bDrawDebug=True,bDraw2D=False,bDrawLocalAxes=False,DebugScale=25,DebugSegments=20,DebugThickness=0.25)
    DriverAxis=(X=1,Y=0,Z=0)
    DriverItem=(Type=Bone,Name="thigh_r")
    FalloffRegionScaleFactors=(PositiveWidth=1,NegativeWidth=1,PositiveHeight=1,NegativeHeight=1)
    FalloffSize=0.3
    FlipHeightScaling=False
    FlipWidthScaling=False
    OptionalParentItem=(Type=Bone,Name="None")
    RotationOffset=(X=90,Y=0,Z=-30)
   )
   > &RigUnit_ModifyTransforms(
    ItemToModify="((Item=(Type=Bone,Name=\"dyn_rear_cloth_r_MSTR\"),Transform=(Rotation=(X=0.0818,Y=0.1241,Z=-0.5284,W=0.8359),Translation=(X=1.2013,Y=-0.1044,Z=1.868),Scale3D=(X=1,Y=1,Z=1))),(Item=(Type=Bone,Name=\"dyn_rear_cloth_c_MSTR\"),Transform=(Rotation=(X=0,Y=0,Z=-0.422 ...(truncated)"
    WeightMinimum=0
    Weight=@n2.OutputParam
   )
   Add = &RigVMFunction_MathDoubleAdd(B=@n1.OutputParam, A=@n2.OutputParam)
   > &RigUnit_ModifyTransforms(
    ItemToModify=((Item=(Type=Bone,Name="dyn_rear_cloth_c_MSTR"),Transform=(Rotation=(X=0,Y=0,Z=0.2164,W=0.9763),Translation=(X=0,Y=0,Z=0),Scale3D=(X=1,Y=1,Z=1))))
    WeightMinimum=0
    Weight=@Add.Result
   )
   &RigUnit_ModifyTransforms(
    ItemToModify="((Item=(Type=Bone,Name=\"dyn_rear_cloth_l_MSTR\"),Transform=(Rotation=(X=0,Y=0,Z=0,W=1),Translation=(X=0,Y=-1,Z=0),Scale3D=(X=1,Y=1,Z=1))),(Item=(Type=Bone,Name=\"dyn_rear_cloth_c_MSTR\"),Transform=(Rotation=(X=0,Y=0,Z=0,W=1),Translation=(X=0,Y=-1,Z=0),Scale3D ...(truncated)"
    Weight=1
    WeightMinimum=0
    ExecutePin=@RigVMFunction_Sequence.C
   )
   RigVMFunctionReturnNode(ExecuteContext=@RigVMFunction_Sequence.D)
  ]}
 New Function_ContainedGraph {nodes [
   RigVMFunctionEntryNode()
   > &RigUnit_ParentConstraint(
    AdvancedSettings=(RotationOrderForFilter=XZY)
    Child=(Type=Bone,Name="dyn_thigh_l_zipper_a_MSTR")
    Parents=((Item=(Type=Bone,Name="pelvis"),Weight=0.0563),(Item=(Type=Bone,Name="deform_groin_l"),Weight=0.0656),(Item=(Type=Bone,Name="thigh_l"),Weight=0.1165),(Item=(Type=Bone,Name="thigh_twist_01_l"),Weight=0.7616))
    Weight=1
   )
   > &RigUnit_ParentConstraint(
    AdvancedSettings=(RotationOrderForFilter=XZY)
    Child=(Type=Bone,Name="dyn_thigh_l_zipper_b_MSTR")
    Parents=((Item=(Type=Bone,Name="deform_knee_out_l"),Weight=0.0161),(Item=(Type=Bone,Name="calf_l"),Weight=0.0175),(Item=(Type=Bone,Name="thigh_twist_01_l"),Weight=0.3431),(Item=(Type=Bone,Name="thigh_l"),Weight=0.6233))
    Weight=1
   )
   > &RigUnit_ParentConstraint(
    AdvancedSettings=(RotationOrderForFilter=XZY)
    Child=(Type=Bone,Name="dyn_thigh_r_zipper_a_MSTR")
    Parents=((Item=(Type=Bone,Name="pelvis"),Weight=0.0576),(Item=(Type=Bone,Name="deform_groin_r"),Weight=0.0674),(Item=(Type=Bone,Name="thigh_r"),Weight=0.1135),(Item=(Type=Bone,Name="thigh_twist_01_r"),Weight=0.7614))
    Weight=1
   )
   > &RigUnit_ParentConstraint(
    AdvancedSettings=(RotationOrderForFilter=XZY)
    Child=(Type=Bone,Name="dyn_thigh_r_zipper_b_MSTR")
    Parents=((Item=(Type=Bone,Name="deform_knee_out_r"),Weight=0.0148),(Item=(Type=Bone,Name="calf_r"),Weight=0.016),(Item=(Type=Bone,Name="thigh_twist_01_r"),Weight=0.3588),(Item=(Type=Bone,Name="thigh_r"),Weight=0.6104))
    Weight=1
   )
   > RigVMFunctionReturnNode()
  ]}
 New Function_ContainedGraph {nodes [
   RigVMFunctionEntryNode()
   > n1 = &RigUnit_SphericalPoseReader(
    ActiveRegionScaleFactors=(PositiveWidth=2.5,NegativeWidth=3,PositiveHeight=3,NegativeHeight=3)
    ActiveRegionSize=0.07
    Debug=(bDrawDebug=false,DebugScale=25,DebugSegments=20,DebugThickness=0.25)
    DriverAxis=(X=-1,Y=0,Z=0)
    DriverItem=(Type=Bone,Name="upperarm_r")
    FalloffRegionScaleFactors=(PositiveWidth=1,NegativeWidth=1,PositiveHeight=1,NegativeHeight=1)
    FalloffSize=0.5
    FlipHeightScaling=true
    OptionalParentItem=(Type=Bone)
    RotationOffset=(X=180,Y=-35,Z=140)
   )
   > &RigUnit_ModifyTransforms(
    ItemToModify="((Item=(Type=Bone,Name=\"deform_upperarm_fr_r\"),Transform=(Rotation=(X=0,Y=0,Z=0.3827,W=0.9239),Translation=(X=4,Y=0,Z=5),Scale3D=(X=1.2,Y=1.2,Z=1.2))),(Item=(Type=Bone,Name=\"upperarm_twist_01_r\"),Transform=(Rotation=(X=0.0436,Y=0,Z=0,W=0.999),Translation=( ...(truncated)"
    WeightMinimum=0
    Weight=@n1.OutputParam
   )
   > RigVMFunctionReturnNode()
  ]}
 New Function_ContainedGraph {nodes [
   RigVMFunctionEntryNode()
   Get_Animation_Attribute = RigVMDispatchNode(BoneName=root, Default=0, Name=deformRig_l_thigh_up_alpha) {template="DISPATCH_RigDispatch_GetAnimAttribute(in Name,in BoneName,in Default,out Value,out Found)"}
   Cast = &RigVMFunction_MathBoolToFloat(Value=RigVMVariableNode() {variable=IsMoving}.Value)
   Subtract = &RigVMFunction_MathDoubleSub(A=1, B=@Cast.Result)
   AlphaInterp = &RigVMFunction_AlphaInterp(InterpSpeedDecreasing=10, InterpSpeedIncreasing=10, Scale=1, bClampResult=False, Value=@Subtract.Result)
   Multiply = &RigVMFunction_MathDoubleMul(A=@Get_Animation_Attribute.Value, B=@AlphaInterp.Result)
   > &RigUnit_ModifyTransforms(
    ItemToModify=((Item=(Type=Bone,Name="CR_l_sash_aimat"),Transform=(Rotation=(X=0.6384,Y=0,Z=0,W=0.7697),Translation=(X=4.1802,Y=-11.4967,Z=2.6756),Scale3D=(X=1,Y=1,Z=1))))
    WeightMinimum=0
    Weight=@Multiply.Result
   )
   Get_Animation_Attribute_1 = RigVMDispatchNode(BoneName=root, Default=0, Name=deformRig_r_thigh_up_alpha) {template="DISPATCH_RigDispatch_GetAnimAttribute(in Name,in BoneName,in Default,out Value,out Found)"}
   > &RigUnit_ModifyTransforms(
    ItemToModify=((Item=(Type=Bone,Name="CR_r_sash_aimat"),Transform=(Rotation=(X=0.6906,Y=0,Z=0,W=0.7232),Translation=(X=-16.6469,Y=0,Z=-7.2574),Scale3D=(X=1,Y=1,Z=1))))
    WeightMinimum=0
    Weight=&RigVMFunction_MathDoubleMul(A=@Get_Animation_Attribute_1.Value, B=@AlphaInterp.Result).Result
   )
   > &RigUnit_AimConstraintLocalSpaceOffset(
    Child=(Type=Bone,Name="dyn_sash_l_MSTR")
    Filter=(bX=True,bY=True,bZ=True)
    Parents=((Item=(Type=Bone,Name="CR_l_sash_aimat"),Weight=1))
    Weight=1
    WorldUp=(Target=(X=0,Y=0,Z=1),Kind=Location,Space=(Type=Bone,Name="CR_l_sash_aimat"))
   )
   > &RigUnit_AimConstraintLocalSpaceOffset(
    Child=(Type=Bone,Name="dyn_sash_r_MSTR")
    Filter=(bX=True,bY=True,bZ=True)
    Parents=((Item=(Type=Bone,Name="CR_r_sash_aimat"),Weight=1))
    Weight=1
    WorldUp=(Target=(X=0,Y=0,Z=1),Kind=Location,Space=(Type=Bone,Name="CR_r_sash_aimat"))
   )
   > RigVMFunctionReturnNode()
  ]}
 RigVMFunctionLibrary {nodes [
   Deform Ankles = RigVMCollapseNode(ExecuteContext=())
   Deform Dynamic mstr bones = RigVMCollapseNode(ExecuteContext=())
   Deform Front Cloth = RigVMCollapseNode(ExecuteContext=())
   Deform Shoulderstraps = RigVMCollapseNode(ExecuteContext=())
   Deformation Rig Extra = RigVMCollapseNode(ExecuteContext=())
   Pin Zippers to Body = RigVMCollapseNode(ExecuteContext=())
  ]}
 RigVMFunction_Sequence_1_ContainedGraph {nodes [
   n1 = RigVMFunctionEntryNode()
   RigVMFunction_Sequence = &RigVMFunction_Sequence(ExecuteContext=@n1.ExecuteContext)
   n2 = &RigVMFunction_Sequence(ExecuteContext=@RigVMFunction_Sequence.B)
   RigVMFunctionReturnNode(A=@RigVMFunction_Sequence.A, B=@n2.A, C=@n2.B)
  ]}
 RigVMFunction_Sequence_1_ContainedGraph {nodes [
   n1 = RigVMFunctionEntryNode()
   RigVMFunction_Sequence = &RigVMFunction_Sequence(ExecuteContext=@n1.ExecuteContext)
   n2 = &RigVMFunction_Sequence(ExecuteContext=@RigVMFunction_Sequence.B)
   RigVMFunctionReturnNode(A=@RigVMFunction_Sequence.A, B=@n2.A, C=@n2.B)
  ]}
 RigVMFunction_Sequence_1_ContainedGraph {nodes [
   n1 = RigVMFunctionEntryNode()
   RigVMFunction_Sequence = &RigVMFunction_Sequence(ExecuteContext=@n1.ExecuteContext)
   n2 = &RigVMFunction_Sequence(ExecuteContext=@RigVMFunction_Sequence.B)
   n3 = &RigVMFunction_Sequence(ExecuteContext=@n2.B)
   RigVMFunctionReturnNode(A=@RigVMFunction_Sequence.A, B=@n2.A, C=@n3.A, D=@n3.B)
  ]}
 RigVMModel {nodes [
   RigUnit_BeginExecution = &RigUnit_BeginExecution()
   FN_DeformRig_FNC = RigVMFunctionReferenceNode(
    DeformRig_ Curve="(deformRig_l_upperarm_up_alpha_72_#1=False,deformRig_r_upperarm_up_alpha_73_#2=False,deformRig_l_upperarm_fr_alpha_74_#3=False,deformRig_r_upperarm_fr_alpha_75_CB51B8 ...(truncated)"
    DeformRig_Bool="(upperarm_1_#4=True,elbow_3_#5=True,wrist_7_#6=True,groin_8_#7=True,glute_17_#8=False,knee_9_40C408324CA26 ...(truncated)"
    DeformRig_Debug=(upperarm_debug_18_#4=False,wrist_debug_20_#6=False,lat_debug_26_#9=False,groinGlute_debug_27_#7=False)
    DeformRig_Value="(upperarm_up_18_#10=(Rotation=(X=0,Y=0,Z=0,W=1),Translation=(X=2,Y=0,Z=4),Scale3D=(X=0.7,Y=1,Z=1)),upperarm_fwd_fr_19_#11=(Rotation=(X=0,Y=0,Z=0,W=1),Translation=(X=0,Y=5,Z=0),Scale3D=(X=1,Y=1,Z=1)) ...(truncated)"
    ExecuteContext=@RigUnit_BeginExecution.ExecutePin
   )
   > Deformation Rig Extra = RigVMFunctionReferenceNode()
   > Deform Shoulderstraps = RigVMFunctionReferenceNode()
   > Deform Ankles = RigVMFunctionReferenceNode()
   > Deform Dynamic mstr bones = RigVMFunctionReferenceNode()
   > Pin Zippers to Body = RigVMFunctionReferenceNode()
   > Deform Front Cloth = RigVMFunctionReferenceNode()
  ]}
]
hierarchy [
 root {loc=<0 0 0>, rot=<0 0 0 1>}
 pelvis {parent=root, loc=<0 -0.7071 93.1191>, rot=<0.0127 0.707 -0.0127 -0.707>}
 spine_01 {parent=pelvis, loc=<6.7494 -0.6388 0>, rot=<0 0 0.034 -0.9994>}
 spine_02 {parent=spine_01, loc=<7.252 -0.035 0>, rot=<0 0 0.0238 -0.9997>}
 spine_03 {parent=spine_02, loc=<6.0106 0 0>, rot=<0 0 0.0149 -0.9999>}
 spine_04 {parent=spine_03, loc=<8.1023 0 0>, rot=<0 0 -0.0132 -0.9999>}
 spine_05 {parent=spine_04, loc=<10.0142 0 0>, rot=<0 0 -0.0015 -1>}
 attach_backpack {parent=spine_05, loc=<-0.123 -13.473 0>, rot=<-0.0093 -0.707 0.0093 -0.707>}
 neck_01 {parent=spine_05, loc=<11.3138 0.1529 0>, rot=<0 0 -0.2003 -0.9797>}
 neck_02 {parent=neck_01, loc=<4.5627 0 0>, rot=<0 0 0.0504 -0.9987>}
 head {parent=neck_02, loc=<5.3914 0 0>, rot=<0 0 0.115 -0.9934>}
 clavicle_l {parent=spine_05, loc=<4.3575 -0.1466 -2.3223>, rot=<0.0465 -0.7289 -0.009 -0.6829>}
 upperarm_l {parent=clavicle_l, loc=<15.5111 0 0>, rot=<0.0226 -0.3234 -0.0088 -0.9459>}
 lowerarm_l {parent=upperarm_l, loc=<26.6514 0 0>, rot=<0.0391 -0.0099 -0.2453 -0.9686>}
 hand_l {parent=lowerarm_l, loc=<22.3491 0 0>, rot=<-0.7081 -0.0507 0.0975 -0.6975>}
 thumb_01_l {parent=hand_l, loc=<1.3118 -0.3191 -2.459>, rot=<0.6546 -0.4042 0.0664 -0.6354>}
 thumb_02_l {parent=thumb_01_l, loc=<5.9336 -0.0448 0.0401>, rot=<-0.0037 0.0213 0.298 -0.9543>}
 thumb_03_l {parent=thumb_02_l, loc=<2.2223 0.3682 -0.0907>, rot=<0.0118 -0.013 0.0066 -0.9998>}
 index_metacarpal_l {parent=hand_l, loc=<1.6535 0.0183 -2.96>, rot=<0.0158 -0.0124 -0.0173 -0.9996>}
 index_01_l {parent=index_metacarpal_l, loc=<8.3008 0.0357 -0.0524>, rot=<-0.0044 0.0485 0.0924 -0.9945>}
 index_02_l {parent=index_01_l, loc=<3.3587 -0.3004 0.0452>, rot=<-0.002 0.0027 0.3225 -0.9466>}
 index_03_l {parent=index_02_l, loc=<2.5578 -0.123 -0.0252>, rot=<-0.002 0.0067 0.2558 -0.9667>}
 middle_metacarpal_l {parent=hand_l, loc=<2.1856 -0.1725 -0.5984>, rot=<0.006 0.0168 -0.0329 -0.9993>}
 middle_01_l {parent=middle_metacarpal_l, loc=<7.3275 -0.2447 0.0129>, rot=<-0.0208 0.0042 0.0858 -0.9961>}
 middle_02_l {parent=middle_01_l, loc=<4.0123 0.1401 0.1175>, rot=<-0.0218 0.0258 0.3512 -0.9357>}
 middle_03_l {parent=middle_02_l, loc=<2.6077 0.0534 0.1373>, rot=<0.0082 0.0188 0.1708 -0.9851>}
 ring_metacarpal_l {parent=hand_l, loc=<2.2074 -0.3625 1.3638>, rot=<-0.0284 0.0722 -0.0458 -0.9959>}
 ring_01_l {parent=ring_metacarpal_l, loc=<7.0295 -0.6197 -0.1446>, rot=<-0.0489 -0.0252 0.1892 -0.9804>}
 ring_02_l {parent=ring_01_l, loc=<3.6174 0.4647 -0.1027>, rot=<0 0 0.2806 -0.9598>}
 ring_03_l {parent=ring_02_l, loc=<2.4152 0.114 0.0199>, rot=<0 0 0.1189 -0.9929>}
 pinky_metacarpal_l {parent=hand_l, loc=<1.7616 -0.9753 3.3814>, rot=<-0.0397 0.0814 -0.069 -0.9935>}
 pinky_01_l {parent=pinky_metacarpal_l, loc=<7.1709 -0.4831 -0.1094>, rot=<-0.0654 -0.0035 0.2242 -0.9723>}
 pinky_02_l {parent=pinky_01_l, loc=<2.8274 0.0893 -0.1043>, rot=<0.0037 0.0011 0.2854 -0.9584>}
 pinky_03_l {parent=pinky_02_l, loc=<2.2046 0.0307 -0.0078>, rot=<-0.0002 0.0098 0.1445 -0.9895>}
 deform_wrist_dn_base_l {parent=hand_l, loc=<0 -2.2342 0>, rot=<0 0 0 -1>}
 deform_wrist_dn_l {parent=deform_wrist_dn_base_l, loc=<0 0 0>, rot=<0 0 0 1>}
 deform_wrist_up_base_l {parent=hand_l, loc=<0 2.2682 0>, rot=<0 0 0 -1>}
 deform_wrist_up_l {parent=deform_wrist_up_base_l, loc=<0 0 0>, rot=<0 0 0 1>}
 lowerarm_twist_02_l {parent=lowerarm_l, loc=<9.6199 0 0>, rot=<0 0 0 -1>}
 lowerarm_twist_01_l {parent=lowerarm_l, loc=<17.8549 0 0>, rot=<0 0 0 -1>}
 deform_elbow_in_base_l {parent=lowerarm_l, loc=<0.2112 5.259 0.8274>, rot=<0 0 0 -1>}
 deform_elbow_in_l {parent=deform_elbow_in_base_l, loc=<0 0 0>, rot=<0 0 0 1>}
 deform_elbow_out_base_l {parent=lowerarm_l, loc=<0 -4 0>, rot=<0 0 0 -1>}
 deform_elbow_out_l {parent=deform_elbow_out_base_l, loc=<0 0 0>, rot=<0 0 0 1>}
 upperarm_twist_01_l {parent=upperarm_l, loc=<0.0041 0 0>, rot=<0 0 0 -1>}
 deform_upperarm_bk_base_l {parent=upperarm_twist_01_l, loc=<1.0813 -7.2193 -1.9453>, rot=<0 0 0 1>}
 deform_upperarm_bk_l {parent=deform_upperarm_bk_base_l, loc=<0 0 0>, rot=<0 0 0 1>}
 deform_upperarm_fr_base_l {parent=upperarm_twist_01_l, loc=<-0.3691 7.0506 -2.765>, rot=<0 0 0 1>}
 deform_upperarm_fr_l {parent=deform_upperarm_fr_base_l, loc=<0 0 0>, rot=<0 0 0 1>}
 deform_upperarm_up_base_l {parent=upperarm_twist_01_l, loc=<-3.2216 -1.1391 3.3311>, rot=<0 0 0 1>}
 deform_upperarm_up_l {parent=deform_upperarm_up_base_l, loc=<0 0 0>, rot=<0 0 0 1>}
 upperarm_twist_02_l {parent=upperarm_l, loc=<13.7941 0 0>, rot=<0 0 0 -1>}
 attach_cape {parent=spine_05, loc=<9.017 -11.714 0>, rot=<0.0616 -0.7044 -0.0616 -0.7044>}
 clavicle_r {parent=spine_05, loc=<4.3575 -0.1466 2.3223>, rot=<0.7289 0.0465 -0.6829 0.009>}
 upperarm_r {parent=clavicle_r, loc=<-15.5111 0 0>, rot=<0.0226 -0.3234 -0.0088 -0.9459>}
 lowerarm_r {parent=upperarm_r, loc=<-26.6514 0 0>, rot=<0.0391 -0.0099 -0.2453 -0.9686>}
 hand_r {parent=lowerarm_r, loc=<-22.3491 0 0>, rot=<-0.7081 -0.0507 0.0975 -0.6975>}
 thumb_01_r {parent=hand_r, loc=<-1.3118 0.3191 2.459>, rot=<0.6546 -0.4042 0.0664 -0.6354>}
 thumb_02_r {parent=thumb_01_r, loc=<-5.9336 0.0448 -0.0401>, rot=<-0.0037 0.0213 0.298 -0.9543>}
 thumb_03_r {parent=thumb_02_r, loc=<-2.2223 -0.3682 0.0907>, rot=<0.0118 -0.013 0.0066 -0.9998>}
 index_metacarpal_r {parent=hand_r, loc=<-1.6535 -0.0183 2.96>, rot=<0.0158 -0.0124 -0.0173 -0.9996>}
 index_01_r {parent=index_metacarpal_r, loc=<-8.3008 -0.0357 0.0524>, rot=<-0.0044 0.0485 0.0924 -0.9945>}
 index_02_r {parent=index_01_r, loc=<-3.3587 0.3004 -0.0452>, rot=<-0.002 0.0027 0.3225 -0.9466>}
 index_03_r {parent=index_02_r, loc=<-2.5578 0.123 0.0252>, rot=<-0.002 0.0067 0.2558 -0.9667>}
 middle_metacarpal_r {parent=hand_r, loc=<-2.1856 0.1725 0.5984>, rot=<0.006 0.0168 -0.0329 -0.9993>}
 middle_01_r {parent=middle_metacarpal_r, loc=<-7.3275 0.2447 -0.0129>, rot=<-0.0208 0.0042 0.0858 -0.9961>}
 middle_02_r {parent=middle_01_r, loc=<-4.0123 -0.1401 -0.1175>, rot=<-0.0218 0.0258 0.3512 -0.9357>}
 middle_03_r {parent=middle_02_r, loc=<-2.6077 -0.0534 -0.1373>, rot=<0.0082 0.0188 0.1708 -0.9851>}
 ring_metacarpal_r {parent=hand_r, loc=<-2.2074 0.3625 -1.3638>, rot=<-0.0284 0.0722 -0.0458 -0.9959>}
 ring_01_r {parent=ring_metacarpal_r, loc=<-7.0295 0.6197 0.1446>, rot=<-0.0489 -0.0252 0.1892 -0.9804>}
 ring_02_r {parent=ring_01_r, loc=<-3.6174 -0.4647 0.1027>, rot=<0 0 0.2806 -0.9598>}
 ring_03_r {parent=ring_02_r, loc=<-2.4152 -0.114 -0.0199>, rot=<0 0 0.1189 -0.9929>}
 pinky_metacarpal_r {parent=hand_r, loc=<-1.7616 0.9753 -3.3814>, rot=<-0.0397 0.0814 -0.069 -0.9935>}
 pinky_01_r {parent=pinky_metacarpal_r, loc=<-7.1709 0.4831 0.1094>, rot=<-0.0654 -0.0035 0.2242 -0.9723>}
 pinky_02_r {parent=pinky_01_r, loc=<-2.8274 -0.0893 0.1043>, rot=<0.0037 0.0011 0.2854 -0.9584>}
 pinky_03_r {parent=pinky_02_r, loc=<-2.2046 -0.0307 0.0078>, rot=<-0.0002 0.0098 0.1445 -0.9895>}
 deform_wrist_dn_base_r {parent=hand_r, loc=<-0.0004 2.2345 -0.0001>, rot=<0 0 0 -1>}
 deform_wrist_dn_r {parent=deform_wrist_dn_base_r, loc=<0 0 0>, rot=<0 0 0 1>}
 deform_wrist_up_base_r {parent=hand_r, loc=<0.0003 -2.2684 0.0001>, rot=<0 0 0 -1>}
 deform_wrist_up_r {parent=deform_wrist_up_base_r, loc=<0 0 0>, rot=<0 0 0 1>}
 lowerarm_twist_02_r {parent=lowerarm_r, loc=<-9.6199 0 0>, rot=<0 0 0 -1>}
 lowerarm_twist_01_r {parent=lowerarm_r, loc=<-17.8549 0 0>, rot=<0 0 0 -1>}
 deform_elbow_in_base_r {parent=lowerarm_r, loc=<-0.2111 -5.2591 -0.8275>, rot=<0 0 0 -1>}
 deform_elbow_in_r {parent=deform_elbow_in_base_r, loc=<-0.0001 0 0.0002>, rot=<0 0 0 1>}
 deform_elbow_out_base_r {parent=lowerarm_r, loc=<0 4 -0.0001>, rot=<0 0 0 -1>}
 deform_elbow_out_r {parent=deform_elbow_out_base_r, loc=<0 0 0>, rot=<0 0 0 1>}
 upperarm_twist_01_r {parent=upperarm_r, loc=<-0.0041 0 0>, rot=<0 0 0 -1>}
 deform_upperarm_bk_base_r {parent=upperarm_twist_01_r, loc=<-1.0814 7.2192 1.9455>, rot=<0 0 0 1>}
 deform_upperarm_bk_r {parent=deform_upperarm_bk_base_r, loc=<0 0 0>, rot=<0 0 0 1>}
 deform_upperarm_fr_base_r {parent=upperarm_twist_01_r, loc=<0.3689 -7.0506 2.7652>, rot=<0 0 0 1>}
 deform_upperarm_fr_r {parent=deform_upperarm_fr_base_r, loc=<0 0 0>, rot=<0 0 0 1>}
 deform_upperarm_up_base_r {parent=upperarm_twist_01_r, loc=<3.2214 1.1391 -3.3308>, rot=<0 0 0 1>}
 deform_upperarm_up_r {parent=deform_upperarm_up_base_r, loc=<0 0 0>, rot=<0 0 0 -1>}
 upperarm_twist_02_r {parent=upperarm_r, loc=<-13.7941 0 0>, rot=<0 0 0 -1>}
 deform_armor_strap_base_l {parent=spine_05, loc=<12.3023 1.1787 -9.788>, rot=<0.1547 -0.7683 -0.1411 -0.6049>}
 deform_armor_strap_l {parent=deform_armor_strap_base_l, loc=<0 0 0>, rot=<0 0 0 1>}
 deform_armor_strap_base_r {parent=spine_05, loc=<12.3021 1.1786 9.788>, rot=<0.7683 0.1547 -0.6049 0.1411>}
 deform_armor_strap_r {parent=deform_armor_strap_base_r, loc=<0 0 0>, rot=<0 0 0 1>}
 deform_pec_aim_base_l {parent=spine_04, loc=<7.8728 12.6388 -2.4123>, rot=<0.0102 -0.4602 0.3979 -0.7936>}
 deform_pec_aim_l {parent=deform_pec_aim_base_l, loc=<0 0 0>, rot=<0 0 0 1>}
 deform_pec_base_l {parent=deform_pec_aim_l, loc=<10.766 0 0>, rot=<0 0 0 1>}
 deform_pec_l {parent=deform_pec_base_l, loc=<0.055 0 0>, rot=<0 0 0 1>}
 deform_pec_aim_base_r {parent=spine_04, loc=<7.8723 12.6388 2.4123>, rot=<-0.4602 -0.0102 0.7936 0.3979>}
 deform_pec_aim_r {parent=deform_pec_aim_base_r, loc=<0 0 0>, rot=<0 0 0 1>}
 deform_pec_base_r {parent=deform_pec_aim_r, loc=<-10.7662 -0.0006 -0.0007>, rot=<0 0 0 1>}
 deform_pec_r {parent=deform_pec_base_r, loc=<0 0 0>, rot=<0 0 0 1>}
 thigh_l {parent=pelvis, loc=<-2.5693 -0.0109 -7.7501>, rot=<0.0012 0.0572 0.0155 -0.9982>}
 calf_l {parent=thigh_l, loc=<-39.0484 0 0>, rot=<0.0006 -0.0161 -0.0391 -0.9991>}
 foot_l {parent=calf_l, loc=<-42.561 0.1077 0.4292>, rot=<0.0418 -0.0455 0.0343 -0.9975>}
 ball_l {parent=foot_l, loc=<-6.1599 13.8264 0.6806>, rot=<0.0001 -0.0001 -0.7063 -0.7079>}
 deform_shoe_rear_base_l {parent=foot_l, loc=<-0.6293 -3.9104 0.2107>, rot=<0.0546 0 0 -0.9985>}
 deform_shoe_rear_l {parent=deform_shoe_rear_base_l, loc=<0 0 0>, rot=<0 0 0 1>}
 dyn_shoe_tag_l {parent=deform_shoe_rear_l, loc=<1.8374 -1.1619 0.0812>, rot=<0.0199 0.0044 0.333 -0.9427>}
 calf_twist_02_l {parent=calf_l, loc=<-18.8064 -0.0967 0.0397>, rot=<0.0237 0.0095 -0.0035 -0.9997>}
 calf_twist_01_l {parent=calf_l, loc=<-35.813 -0.0967 0.0397>, rot=<0.0237 0.0095 -0.0035 -0.9997>}
 deform_knee_in_base_l {parent=calf_l, loc=<1.298 -5.2957 -0.0496>, rot=<0 0 0 -1>}
 deform_knee_in_l {parent=deform_knee_in_base_l, loc=<0 0 0>, rot=<0 0 0 1>}
 deform_knee_out_base_l {parent=calf_l, loc=<0 4 0>, rot=<0 0 0 -1>}
 deform_knee_out_l {parent=deform_knee_out_base_l, loc=<0 0 0>, rot=<0 0 0 1>}
 thigh_twist_01_l {parent=thigh_l, loc=<-21.6303 0 0>, rot=<0.0013 0 0 -1>}
 dyn_thigh_l_zipper_a_MSTR {parent=thigh_twist_01_l, loc=<6.1283 5.1224 -8.6979>, rot=<-0.0266 -0.0123 0.9992 0.0282>}
 dyn_thigh_l_zipper_a {parent=dyn_thigh_l_zipper_a_MSTR, loc=<0 0 0>, rot=<0 0 0 1>}
 dyn_thigh_l_zipper_b_MSTR {parent=thigh_twist_01_l, loc=<-8.7414 1.4396 -9.6687>, rot=<0.0281 0.6323 0.7738 -0.0268>}
 dyn_thigh_l_zipper_b {parent=dyn_thigh_l_zipper_b_MSTR, loc=<0 0 0>, rot=<0 0 0 1>}
 deform_groin_base_l {parent=thigh_l, loc=<-2.1908 8.1142 -1.7442>, rot=<0 0 0 -1>}
 deform_groin_l {parent=deform_groin_base_l, loc=<0 0 0>, rot=<0 0 0 1>}
 CR_l_sash_aimat {parent=thigh_l, loc=<-39.0484 0 0>, rot=<0.0006 -0.0161 -0.0391 -0.9991>}
 thigh_r {parent=pelvis, loc=<-2.5693 -0.0109 7.7501>, rot=<0.0572 -0.0012 0.9982 0.0155>}
 calf_r {parent=thigh_r, loc=<39.0484 0 0>, rot=<0.0006 -0.0161 -0.0391 -0.9991>}
 foot_r {parent=calf_r, loc=<42.561 -0.1077 -0.4292>, rot=<0.0418 -0.0455 0.0343 -0.9975>}
 ball_r {parent=foot_r, loc=<6.1599 -13.8264 -0.6806>, rot=<0.0001 -0.0001 -0.7063 -0.7079>}
 deform_shoe_rear_base_r {parent=foot_r, loc=<0.6293 3.9104 -0.2107>, rot=<0.0546 0 0 -0.9985>}
 deform_shoe_rear_r {parent=deform_shoe_rear_base_r, loc=<0 0 0>, rot=<0 0 0 -1>}
 dyn_shoe_tag_r {parent=deform_shoe_rear_r, loc=<-1.8374 1.1619 -0.0812>, rot=<0.0199 0.0044 0.333 -0.9427>}
 calf_twist_02_r {parent=calf_r, loc=<18.8064 0.0967 -0.0397>, rot=<0.0237 0.0095 -0.0035 -0.9997>}
 calf_twist_01_r {parent=calf_r, loc=<35.813 0.0967 -0.0397>, rot=<0.0237 0.0095 -0.0035 -0.9997>}
 deform_knee_in_base_r {parent=calf_r, loc=<-1.298 5.2957 0.0496>, rot=<0 0 0 -1>}
 deform_knee_in_r {parent=deform_knee_in_base_r, loc=<0 0 0>, rot=<0 0 0 1>}
 deform_knee_out_base_r {parent=calf_r, loc=<0 -4 0>, rot=<0 0 0 -1>}
 deform_knee_out_r {parent=deform_knee_out_base_r, loc=<0 0 0>, rot=<0 0 0 1>}
 thigh_twist_01_r {parent=thigh_r, loc=<21.6303 0 0>, rot=<0.0013 0 0 -1>}
 dyn_thigh_r_zipper_a_MSTR {parent=thigh_twist_01_r, loc=<-6.1283 -5.1224 8.6979>, rot=<-0.0266 -0.0123 0.9992 0.0282>}
 dyn_thigh_r_zipper_a {parent=dyn_thigh_r_zipper_a_MSTR, loc=<0 0 0>, rot=<0 0 0 1>}
 dyn_thigh_r_zipper_b_MSTR {parent=thigh_twist_01_r, loc=<8.7414 -1.4396 9.6686>, rot=<0.0281 0.6323 0.7738 -0.0268>}
 dyn_thigh_r_zipper_b {parent=dyn_thigh_r_zipper_b_MSTR, loc=<0 0 0>, rot=<0 0 0 1>}
 deform_groin_base_r {parent=thigh_r, loc=<2.1908 -8.1142 1.7442>, rot=<0 0 0 -1>}
 deform_groin_r {parent=deform_groin_base_r, loc=<0 0 0>, rot=<0 0 0 1>}
 CR_r_sash_aimat {parent=thigh_r, loc=<39.0484 0 0>, rot=<0.0006 -0.0161 -0.0391 -0.9991>}
 deform_glute_base_l {parent=pelvis, loc=<-8.8064 -8.8222 -8.4287>, rot=<0.0275 0.0501 -0.4472 -0.8926>}
 deform_glute_l {parent=deform_glute_base_l, loc=<0 0 0>, rot=<0 0 0 1>}
 deform_glute_base_r {parent=pelvis, loc=<-8.8062 -8.8222 8.4287>, rot=<0.0501 -0.0275 0.8926 -0.4472>}
 deform_glute_r {parent=deform_glute_base_r, loc=<0 0 0>, rot=<0 0 0 1>}
 dyn_sash_root {parent=pelvis, loc=<11.1653 10.5324 0.5176>, rot=<0.7196 0.6942 0.011 -0.0107>}
 dyn_sash_l_MSTR {parent=dyn_sash_root, loc=<0.3964 -1.2312 3.2908>, rot=<0.1244 0.1646 0.6527 -0.7289>}
 dyn_sash_l_01 {parent=dyn_sash_l_MSTR, loc=<0 0 0>, rot=<0 0 0 1>}
 dyn_sash_l_02 {parent=dyn_sash_l_01, loc=<5 0 0>, rot=<0.0009 -0.0769 0.0115 -0.997>}
 dyn_sash_l_03 {parent=dyn_sash_l_02, loc=<5 0 0>, rot=<-0.0001 -0.0323 -0.0026 -0.9995>}
 dyn_sash_l_04 {parent=dyn_sash_l_03, loc=<5 0 0>, rot=<0 0 0 1>}
 dyn_sash_l_05 {parent=dyn_sash_l_04, loc=<5 0 0>, rot=<0 0 0 1>}
 dyn_sash_l_06 {parent=dyn_sash_l_02, loc=<6.6273 -1.7861 1.8686>, rot=<-0.2123 0.144 0.0497 -0.9653>}
 dyn_sash_l_07 {parent=dyn_sash_l_06, loc=<5 0 0>, rot=<0 0 0 1>}
 dyn_sash_r_MSTR {parent=dyn_sash_root, loc=<0.3259 -1.0793 -2.164>, rot=<-0.1129 0.1103 -0.7468 -0.646>}
 dyn_sash_r_01 {parent=dyn_sash_r_MSTR, loc=<0 0 0>, rot=<0 0 0 1>}
 dyn_sash_r_02 {parent=dyn_sash_r_01, loc=<-4.9999 0 0>, rot=<0 -0.0607 0 -0.9982>}
 dyn_sash_r_03 {parent=dyn_sash_r_02, loc=<-5 0.0001 0>, rot=<0 -0.0417 0 -0.9991>}
 dyn_sash_r_04 {parent=dyn_sash_r_03, loc=<-5 0 0>, rot=<0 0 0 -1>}
 dyn_sash_r_05 {parent=dyn_sash_r_04, loc=<-5 0 0>, rot=<0 0 0 -1>}
 dyn_sash_r_06 {parent=dyn_sash_r_02, loc=<-7.8558 2.1755 -1.8736>, rot=<-0.2097 0.1199 0.0579 -0.9686>}
 dyn_rear_cloth_l_MSTR {parent=pelvis, loc=<3.2948 -9.8436 -9.4551>, rot=<0.0566 -0.976 -0.2076 -0.0338>}
 dyn_rear_cloth_l_01 {parent=dyn_rear_cloth_l_MSTR, loc=<0 0 0>, rot=<0 0 0 1>}
 dyn_rear_cloth_l_02 {parent=dyn_rear_cloth_l_01, loc=<8 0 0>, rot=<0 0 -0.0519 -0.9987>}
 dyn_rear_cloth_l_03 {parent=dyn_rear_cloth_l_02, loc=<8 0 0>, rot=<0 0 0 1>}
 dyn_rear_cloth_l_04 {parent=dyn_rear_cloth_l_03, loc=<8 0 0>, rot=<0 0 0 1>}
 dyn_rear_cloth_l_05 {parent=dyn_rear_cloth_l_04, loc=<8 0 0>, rot=<0 0 0 1>}
 dyn_rear_cloth_c_MSTR {parent=pelvis, loc=<3.4872 -11.3062 0>, rot=<0.0734 -0.9973 0 0>}
 dyn_rear_cloth_c_01 {parent=dyn_rear_cloth_c_MSTR, loc=<0 0 0>, rot=<0 0 0 1>}
 dyn_rear_cloth_c_02 {parent=dyn_rear_cloth_c_01, loc=<8 0 0>, rot=<0 0 -0.0519 -0.9987>}
 dyn_rear_cloth_c_03 {parent=dyn_rear_cloth_c_02, loc=<8 0 0>, rot=<0 0 0 1>}
 dyn_rear_cloth_c_04 {parent=dyn_rear_cloth_c_03, loc=<8 0 0>, rot=<0 0 0 1>}
 dyn_rear_cloth_c_05 {parent=dyn_rear_cloth_c_04, loc=<8 0 0>, rot=<0 0 0 1>}
 dyn_rear_cloth_extra_l_MSTR {parent=pelvis, loc=<3.6456 -4.8953 -13.6027>, rot=<0.165 -0.8056 -0.5554 -0.124>}
 dyn_rear_cloth_extra_l_01 {parent=dyn_rear_cloth_extra_l_MSTR, loc=<0 0 0>, rot=<0 0 0 1>}
 dyn_rear_cloth_extra_l_02 {parent=dyn_rear_cloth_extra_l_01, loc=<8 0 0>, rot=<0 0 -0.0619 -0.9981>}
 dyn_rear_cloth_extra_l_03 {parent=dyn_rear_cloth_extra_l_02, loc=<8 0 0>, rot=<0 0 -0.0173 -0.9999>}
 dyn_rear_cloth_extra_l_04 {parent=dyn_rear_cloth_extra_l_03, loc=<8 0 0>, rot=<0 0 0 1>}
 dyn_rear_cloth_r_MSTR {parent=pelvis, loc=<3.2948 -9.8436 9.4551>, rot=<0.976 0.0566 -0.0338 0.2076>}
 dyn_rear_cloth_r_01 {parent=dyn_rear_cloth_r_MSTR, loc=<0 0 0>, rot=<0 0 0 1>}
 dyn_rear_cloth_r_02 {parent=dyn_rear_cloth_r_01, loc=<-8 0 0>, rot=<0 0 -0.0519 -0.9987>}
 dyn_rear_cloth_r_03 {parent=dyn_rear_cloth_r_02, loc=<-8 0 0>, rot=<0 0 0 -1>}
 dyn_rear_cloth_r_04 {parent=dyn_rear_cloth_r_03, loc=<-8 0 0>, rot=<0 0 0 -1>}
 dyn_rear_cloth_r_05 {parent=dyn_rear_cloth_r_04, loc=<-8 0 0>, rot=<0 0 0 -1>}
 dyn_rear_cloth_extra_r_MSTR {parent=pelvis, loc=<3.6501 -4.6372 13.6999>, rot=<-0.8017 -0.1492 0.1468 -0.5598>}
 dyn_rear_cloth_extra_r_01 {parent=dyn_rear_cloth_extra_r_MSTR, loc=<0 0 0>, rot=<0 0 0 1>}
 dyn_rear_cloth_extra_r_02 {parent=dyn_rear_cloth_extra_r_01, loc=<-7.9999 0 0>, rot=<0 0 -0.0838 -0.9965>}
 dyn_rear_cloth_extra_r_03 {parent=dyn_rear_cloth_extra_r_02, loc=<-8 0.0001 0>, rot=<0 0 -0.0173 -0.9999>}
 dyn_rear_cloth_extra_r_04 {parent=dyn_rear_cloth_extra_r_03, loc=<-8 0 0>, rot=<0 0 0 -1>}
 dyn_rear_cloth_extra_r_05 {parent=dyn_rear_cloth_extra_r_04, loc=<-6.0036 0 0>, rot=<0 0 0 1>}
 dyn_rear_cloth_extra_r_06 {parent=dyn_rear_cloth_extra_r_03, loc=<-0.7075 1.2257 1.0913>, rot=<0 0 0.0227 -0.9997>}
 dyn_rear_cloth_extra_r_07 {parent=dyn_rear_cloth_extra_r_06, loc=<-6.0036 0 0>, rot=<0 0 0 1>}
 dyn_floor_collision {parent=root, loc=<0 0 -50>, rot=<0 0 0 1>}
]
variables [IsMoving {type=bool, default=False}]
