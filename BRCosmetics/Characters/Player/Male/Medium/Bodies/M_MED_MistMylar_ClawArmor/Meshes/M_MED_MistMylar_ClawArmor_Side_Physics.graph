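# Physics asset for the side mesh of the MistMylar ClawArmor medium male player body, covering its dangling rear cloth bones.
# Capsules on pelvis, thighs, calves and feet form the lower body alongside four cloth chains rooted to the pelvis through tiny MSTR bodies.
# Swing and twist limits shape each chain, with the right extra chain forking at its third bone.
# Cross-constraints tie upper extra-chain bones to the main chains, coupling neighboring strips.
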PhysicsAsset /BRCosmetics/Characters/Player/Male/Medium/Bodies/M_MED_MistMylar_ClawArmor/Meshes/M_MED_MistMylar_ClawArmor_Side_Physics
bodies {
 calf_l [{shapeType=Capsule, length=21.6405, radius=7.7475, center=<-20.3227 -1.2278 -0.0878>, rotation=<rot 0 -91.0668 -89.1923>}]
 calf_r [{shapeType=Capsule, length=21.5588, radius=7.9185, center=<19.9158 1.4614 -0.005>, rotation=<rot 0 -90.9209 -89.2104>}]
 dyn_rear_cloth_extra_l_01 [{shapeType=Capsule, length=3.1845, radius=3.8653, shapeName=dyn_rear_cloth_extra_l_01_capsule, center=<4 -4 0>}]
 dyn_rear_cloth_extra_l_02 [{shapeType=Capsule, length=8.9901, radius=4.0219, shapeName=dyn_rear_cloth_extra_l_02_capsule, center=<4 -4 0>}]
 dyn_rear_cloth_extra_l_03 [{shapeType=Capsule, length=9.09, radius=4.4467, shapeName=dyn_rear_cloth_extra_l_03_capsule, center=<4 -4 0>}]
 dyn_rear_cloth_extra_l_04 [{shapeType=Capsule, length=12.8973, radius=4, shapeName=dyn_rear_cloth_extra_l_04_capsule, center=<4 -4 0>}]
 dyn_rear_cloth_extra_l_MSTR [{shapeType=Capsule, length=0.505, radius=0.505, shapeName=dyn_rear_cloth_extra_l_MSTR_capsule, rotation=<rot 0 -90 0>}]
 dyn_rear_cloth_extra_r_01 [{shapeType=Capsule, length=2.2229, radius=4.2421, shapeName=dyn_rear_cloth_extra_r_01_capsule, center=<-4 4 0>}]
 dyn_rear_cloth_extra_r_02 [{shapeType=Capsule, length=4.62, radius=4.2152, shapeName=dyn_rear_cloth_extra_r_02_capsule, center=<-4 4 0>}]
 dyn_rear_cloth_extra_r_03 [{shapeType=Capsule, length=4.4592, radius=3.5755, shapeName=dyn_rear_cloth_extra_r_03_capsule, center=<-4 4 0>}]
 dyn_rear_cloth_extra_r_04 [{shapeType=Capsule, length=3.6798, radius=3.6543, shapeName=dyn_rear_cloth_extra_r_04_capsule, center=<-4 4 0>}]
 dyn_rear_cloth_extra_r_05 [{shapeType=Capsule, length=4.2188, radius=4.6318, shapeName=dyn_rear_cloth_extra_r_05_capsule, center=<-4 4 0>}]
 dyn_rear_cloth_extra_r_06 [{shapeType=Capsule, length=4.3011, radius=3.8021, shapeName=dyn_rear_cloth_extra_r_06_capsule, center=<-3.3921 2.7624 -1.0731>}]
 dyn_rear_cloth_extra_r_07 [{shapeType=Capsule, length=7.1478, radius=5.7089, shapeName=dyn_rear_cloth_extra_r_07_capsule, center=<-4 4 0>}]
 dyn_rear_cloth_extra_r_MSTR [{shapeType=Capsule, length=0.505, radius=0.505, shapeName=dyn_rear_cloth_extra_r_MSTR_capsule}]
 dyn_rear_cloth_l_01 [{shapeType=Capsule, length=8.4319, radius=3.8634, shapeName=dyn_rear_cloth_l_01_capsule, center=<4 -5.2 -0.4375>, rotation=<rot 0 0 25>}]
 dyn_rear_cloth_l_02 [{
   shapeType=Capsule
   length=11.3975
   radius=4.0905
   shapeName=dyn_rear_cloth_l_02_capsule
   center=<4 -1.1426 -0.0699>
   rotation=<rot 0.0043 0.0096 37.8238>
  }]
 dyn_rear_cloth_l_03 [{shapeType=Capsule, length=8.3464, radius=4.35, shapeName=dyn_rear_cloth_l_03_capsule, center=<4 -0.2745 0.5773>, rotation=<rot 0 0.0001 25>}]
 dyn_rear_cloth_l_04 [{
   shapeType=Capsule
   length=5.6332
   radius=4.8615
   shapeName=dyn_rear_cloth_l_04_capsule
   center=<3.9999 -2.1471 1.4674>
   rotation=<rot 0.0011 0.0025 13.9393>
  }]
 dyn_rear_cloth_l_05 [{
   shapeType=Capsule
   length=9.2834
   radius=4.8615
   shapeName=dyn_rear_cloth_l_05_capsule
   center=<4 -1.4026 0.8147>
   rotation=<rot 0.0004 0.0008 8.5615>
  }]
 dyn_rear_cloth_l_MSTR [{shapeType=Capsule, length=0.505, radius=0.505, shapeName=dyn_rear_cloth_l_MSTR_capsule, rotation=<rot 0 -90 0>}]
 dyn_rear_cloth_r_01 [{shapeType=Capsule, length=5.1513, radius=4.3219, shapeName=dyn_rear_cloth_l_01_capsule, center=<-4 5.2 0>, rotation=<rot 0 0 25>}]
 dyn_rear_cloth_r_02 [{shapeType=Capsule, length=5.1388, radius=4.8615, shapeName=dyn_rear_cloth_l_02_capsule, center=<-4 4.25 0>, rotation=<rot 0 0 25>}]
 dyn_rear_cloth_r_03 [{
   shapeType=Capsule
   length=11.8194
   radius=4.8615
   shapeName=dyn_rear_cloth_l_03_capsule
   center=<-4 1.8191 -0.6966>
   rotation=<rot 0 0.0001 35.3959>
  }]
 dyn_rear_cloth_r_04 [{shapeType=Capsule, length=11.9288, radius=4.8615, shapeName=dyn_rear_cloth_l_04_capsule, center=<-4 4.25 0>, rotation=<rot 0 0 25>}]
 dyn_rear_cloth_r_05 [{shapeType=Capsule, length=11.8785, radius=4.8615, shapeName=dyn_rear_cloth_l_05_capsule, center=<-4 4.25 0>, rotation=<rot 0.0003 0.0007 25>}]
 dyn_rear_cloth_r_MSTR [{shapeType=Capsule, length=0.505, radius=0.505, shapeName=dyn_rear_cloth_r_MSTR_capsule, rotation=<rot 0 -90 0>}]
 foot_l [{shapeType=Capsule, length=15.201, radius=9.2891, center=<-3.7807 6.7846 -0.3706>, rotation=<rot 0 17.5747 -89.0234>}]
 foot_r [{shapeType=Capsule, length=15.2015, radius=9.291, center=<3.7846 -6.7828 0.3663>, rotation=<rot 0 17.6088 -88.9869>}]
 pelvis [{shapeType=Capsule, length=11.1804, radius=9.6886, center=<0 -1.4042 0>, rotation=<rot 0 -90 0>}]
 thigh_l [
  {shapeType=Capsule, length=28.604, radius=10.2544, center=<-16.7272 -0.9555 0.5386>, rotation=<rot 0 -88.7531 -96.1754>}
  {
   shapeType=Capsule
   length=5.5473
   radius=9.7804
   shapeName=thigh_l_capsule
   center=<-17.6286 0 -1.6751>
   rotation=<rot 86.0387 179.9995 179.9997>
  }
 ]
 thigh_r [
  {shapeType=Capsule, length=31.986, radius=9.434, center=<16.7555 -0.3428 -0.5386>, rotation=<rot 0 -88.7528 87.0477>}
  {
   shapeType=Capsule
   length=5.1519
   radius=8.6062
   shapeName=thigh_r_capsule
   center=<18.5796 1.1859 2.9775>
   rotation=<rot -87.9103 -0.0004 0.0003>
  }
 ]
}
constraints [
 {bone1Name=thigh_l, bone2Name=pelvis, swing1LimitDegrees=45, swing2LimitDegrees=45, twistLimitDegrees=45}
 {bone1Name=calf_l, bone2Name=thigh_l, swing1LimitDegrees=45, swing2LimitDegrees=45, twistLimitDegrees=45}
 {bone1Name=foot_l, bone2Name=calf_l, swing1LimitDegrees=45, swing2LimitDegrees=45, twistLimitDegrees=45}
 {bone1Name=calf_r, bone2Name=thigh_r, swing1LimitDegrees=45, swing2LimitDegrees=45, twistLimitDegrees=45}
 {bone1Name=foot_r, bone2Name=calf_r, swing1LimitDegrees=45, swing2LimitDegrees=45, twistLimitDegrees=45}
 {
  bone1Name=dyn_rear_cloth_extra_l_MSTR
  bone2Name=pelvis
  swing1LimitDegrees=45
  swing1Motion=Locked
  swing2LimitDegrees=45
  swing2Motion=Locked
  twistLimitDegrees=45
  twistMotion=Locked
 }
 {bone1Name=dyn_rear_cloth_extra_l_01, bone2Name=dyn_rear_cloth_extra_l_MSTR, swing1LimitDegrees=8, swing2LimitDegrees=8, twistLimitDegrees=2}
 {bone1Name=dyn_rear_cloth_extra_l_02, bone2Name=dyn_rear_cloth_extra_l_01, swing1LimitDegrees=24, swing2LimitDegrees=24, twistLimitDegrees=2}
 {bone1Name=dyn_rear_cloth_extra_l_03, bone2Name=dyn_rear_cloth_extra_l_02, swing1LimitDegrees=24, swing2LimitDegrees=24, twistLimitDegrees=2}
 {bone1Name=dyn_rear_cloth_extra_l_04, bone2Name=dyn_rear_cloth_extra_l_03, swing1LimitDegrees=24, swing2LimitDegrees=24, twistLimitDegrees=2}
 {
  bone1Name=dyn_rear_cloth_extra_r_MSTR
  bone2Name=pelvis
  swing1LimitDegrees=45
  swing1Motion=Free
  swing2LimitDegrees=45
  swing2Motion=Free
  twistLimitDegrees=45
  twistMotion=Free
 }
 {bone1Name=dyn_rear_cloth_extra_r_01, bone2Name=dyn_rear_cloth_extra_r_MSTR, swing1LimitDegrees=12, swing2LimitDegrees=8, twistLimitDegrees=4}
 {bone1Name=dyn_rear_cloth_extra_r_02, bone2Name=dyn_rear_cloth_extra_r_01, swing1LimitDegrees=24, swing2LimitDegrees=24, twistLimitDegrees=4}
 {bone1Name=dyn_rear_cloth_extra_r_03, bone2Name=dyn_rear_cloth_extra_r_02, swing1LimitDegrees=12, swing2LimitDegrees=8, twistLimitDegrees=4}
 {bone1Name=dyn_rear_cloth_extra_r_04, bone2Name=dyn_rear_cloth_extra_r_03, swing1LimitDegrees=18, swing2LimitDegrees=12, twistLimitDegrees=4}
 {bone1Name=dyn_rear_cloth_extra_r_05, bone2Name=dyn_rear_cloth_extra_r_04, swing1LimitDegrees=18, swing2LimitDegrees=12, twistLimitDegrees=4}
 {bone1Name=dyn_rear_cloth_extra_r_06, bone2Name=dyn_rear_cloth_extra_r_03, swing1LimitDegrees=18, swing2LimitDegrees=8, twistLimitDegrees=4}
 {bone1Name=dyn_rear_cloth_extra_r_07, bone2Name=dyn_rear_cloth_extra_r_06, swing1LimitDegrees=18, swing2LimitDegrees=8, twistLimitDegrees=4}
 {
  bone1Name=dyn_rear_cloth_l_MSTR
  bone2Name=pelvis
  swing1LimitDegrees=45
  swing1Motion=Free
  swing2LimitDegrees=45
  swing2Motion=Free
  twistLimitDegrees=45
  twistMotion=Free
 }
 {bone1Name=dyn_rear_cloth_l_01, bone2Name=dyn_rear_cloth_l_MSTR, swing1LimitDegrees=10, swing2LimitDegrees=1, twistLimitDegrees=1}
 {bone1Name=dyn_rear_cloth_l_02, bone2Name=dyn_rear_cloth_l_01, swing1LimitDegrees=48, swing2LimitDegrees=24, twistLimitDegrees=12}
 {bone1Name=dyn_rear_cloth_l_03, bone2Name=dyn_rear_cloth_l_02, swing1LimitDegrees=48, swing2LimitDegrees=24, twistLimitDegrees=12}
 {bone1Name=dyn_rear_cloth_l_04, bone2Name=dyn_rear_cloth_l_03, swing1LimitDegrees=48, swing2LimitDegrees=24, twistLimitDegrees=12}
 {bone1Name=dyn_rear_cloth_l_05, bone2Name=dyn_rear_cloth_l_04, swing1LimitDegrees=48, swing2LimitDegrees=24, twistLimitDegrees=12}
 {
  bone1Name=dyn_rear_cloth_r_MSTR
  bone2Name=pelvis
  swing1LimitDegrees=45
  swing1Motion=Free
  swing2LimitDegrees=45
  swing2Motion=Free
  twistLimitDegrees=45
  twistMotion=Free
 }
 {bone1Name=thigh_r, bone2Name=pelvis, swing1LimitDegrees=45, swing2LimitDegrees=45, twistLimitDegrees=45}
 {bone1Name=dyn_rear_cloth_r_01, bone2Name=dyn_rear_cloth_r_MSTR, swing1LimitDegrees=10, swing2LimitDegrees=1, twistLimitDegrees=1}
 {bone1Name=dyn_rear_cloth_r_02, bone2Name=dyn_rear_cloth_r_01, swing1LimitDegrees=48, swing2LimitDegrees=24, twistLimitDegrees=12}
 {bone1Name=dyn_rear_cloth_r_03, bone2Name=dyn_rear_cloth_r_02, swing1LimitDegrees=48, swing2LimitDegrees=24, twistLimitDegrees=12}
 {bone1Name=dyn_rear_cloth_r_04, bone2Name=dyn_rear_cloth_r_03, swing1LimitDegrees=48, swing2LimitDegrees=24, twistLimitDegrees=12}
 {bone1Name=dyn_rear_cloth_r_05, bone2Name=dyn_rear_cloth_r_04, swing1LimitDegrees=48, swing2LimitDegrees=24, twistLimitDegrees=12}
 {bone1Name=dyn_rear_cloth_extra_r_02, bone2Name=dyn_rear_cloth_r_02, swing1LimitDegrees=10, swing2LimitDegrees=10, twistLimitDegrees=10}
 {
  bone1Name=dyn_rear_cloth_extra_r_03
  bone2Name=dyn_rear_cloth_r_03
  swing1LimitDegrees=45
  swing1Motion=Free
  swing2LimitDegrees=45
  swing2Motion=Free
  twistLimitDegrees=45
  twistMotion=Free
 }
 {bone1Name=dyn_rear_cloth_extra_l_02, bone2Name=dyn_rear_cloth_l_02, swing1LimitDegrees=45, swing2LimitDegrees=20, twistLimitDegrees=5}
 {bone1Name=dyn_rear_cloth_extra_r_01, bone2Name=dyn_rear_cloth_r_01, swing1LimitDegrees=10, swing2LimitDegrees=10, twistLimitDegrees=10}
 {bone1Name=dyn_rear_cloth_extra_l_01, bone2Name=dyn_rear_cloth_l_01, swing1LimitDegrees=5, swing2LimitDegrees=5, twistLimitDegrees=5}
]
