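# Cosmetic AnimBP for the MistMylar ClawArmor body, adding secondary motion to cloth, shoe tags and zippers.
# The parent mesh pose runs through a deformation Control Rig, then additive bone offsets on rear cloth and floor collision.
# Two pelvis-space RigidBody passes and cloth nudges follow, then spring AnimDynamics for tags and zippers.
# With rigid bodies enabled, updates set gravity via a deprecated function, plus front-end or sliding simulation settings and simulation space.
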AnimBlueprint /BRCosmetics/Characters/Player/Male/Medium/Bodies/M_MED_MistMylar_ClawArmor/Meshes/M_MED_MistMylar_ClawArmor_AnimBP
target_skeleton=&/M_MED_MistMylar_ClawArmor_Skeleton
graphs [
 AnimGraph {nodes [
   Output Pose(Result=Use cached pose 'Post Extra Dynamics'()) {layerGroup=None}
   n1 = Control Rig!(Source=Copy Pose From Mesh() {bUseAttachedParent=true}, Alpha, IsMoving:bool=Get bIsMoving2D()) -> (Pose) {
    bTransferInputCurves=false
    alphaBoolBlend {blendOption=Linear}
    controlRigAssetReference {blueprintRigClass=&/M_MED_MistMylar_ClawArmor_CtrlRig.M_MED_MistMylar_ClawArmor_CtrlRig_C}
   }
   Save cached pose 'Post Deformation Rig'(Pose=@n1) {cacheName=Post Deformation Rig}
   n2 = RigidBody!(
    ComponentPose=Local To Component(LocalPose=Use cached pose 'Post Pose'())
    Alpha
    OverrideWorldGravity:Vector="0,0,0"
    SimSpaceSettings:Sim Space Settings Structure=(WorldAlpha=1,VelocityScaleZ=1,DampingAlpha=1,MaxLinearVelocity=500,MaxAngularVelocity=8,MaxLinearAcceleration=200,MaxAngularAcceleration=20,ExternalLinearDragV=(X=0.2,Y=0.2,Z=0.2),ExternalLinearVelocity=(X=0,Y=0,Z=0),ExternalAngularVelocity=(X=0,Y=0,Z=0))
   ) -> (Pose) {
    bOverrideWorldGravity=true
    overridePhysicsAsset=&/M_MED_MistMylar_ClawArmor_Physics
    simulationSpace=BaseBoneSpace
    alphaBoolBlend {blendOption=Linear}
    baseBoneRef {boneName=pelvis}
    simSpaceSettings {
     maxAngularAcceleration=20
     maxAngularVelocity=8
     maxLinearAcceleration=200
     maxLinearVelocity=500
     worldAlpha=1
     externalLinearDragV=<0.2 0.2 0.2>
    }
   }
   n3 = RigidBody(
    ComponentPose=@n2
    OverrideWorldGravity="0,0,0"
    SimSpaceSettings=(WorldAlpha=1,VelocityScaleZ=1,DampingAlpha=1,MaxLinearVelocity=500,MaxAngularVelocity=8,MaxLinearAcceleration=200,MaxAngularAcceleration=20,ExternalLinearDragV=(X=0.2,Y=0.2,Z=0.2),ExternalLinearVelocity=(X=0,Y=0,Z=0),ExternalAngularVelocity=(X=0,Y=0,Z=0))
   ) {
    bOverrideWorldGravity=true
    overridePhysicsAsset=&/M_MED_MistMylar_ClawArmor_Side_Physics
    simulationSpace=BaseBoneSpace
    alphaBoolBlend {blendOption=Linear}
    baseBoneRef {boneName=pelvis}
    simSpaceSettings {
     maxAngularAcceleration=20
     maxAngularVelocity=8
     maxLinearAcceleration=200
     maxLinearVelocity=500
     worldAlpha=1
     externalLinearDragV=<0.2 0.2 0.2>
    }
   }
   n4 = "Transform (Modify) Bone - Bone: dyn_rear_cloth_l_MSTR"(ComponentPose=@n3, Alpha=1, Translation="0,-0.5,0", Rotation="0,0,0", Scale="1,1,1") {
    rotationMode=BMM_Additive
    rotationSpace=BCS_BoneSpace
    translationMode=BMM_Additive
    translationSpace=BCS_BoneSpace
    alphaBoolBlend {blendOption=Linear}
    boneToModify {boneName=dyn_rear_cloth_l_MSTR}
    translation {y=-0.5}
   }
   n5 = "Transform (Modify) Bone - Bone: dyn_rear_cloth_c_MSTR"(ComponentPose=@n4, Alpha=1, Translation="0,-0.5,0", Rotation="0,0,0", Scale="1,1,1") {
    rotationMode=BMM_Additive
    rotationSpace=BCS_BoneSpace
    translationMode=BMM_Additive
    translationSpace=BCS_BoneSpace
    alphaBoolBlend {blendOption=Linear}
    boneToModify {boneName=dyn_rear_cloth_c_MSTR}
    translation {y=-0.5}
   }
   n6 = "Transform (Modify) Bone - Bone: dyn_rear_cloth_r_MSTR"(ComponentPose=@n5, Alpha=1, Translation="0,0.5,0", Rotation="0,0,0", Scale="1,1,1") {
    rotationMode=BMM_Additive
    rotationSpace=BCS_BoneSpace
    translationMode=BMM_Additive
    translationSpace=BCS_BoneSpace
    alphaBoolBlend {blendOption=Linear}
    boneToModify {boneName=dyn_rear_cloth_r_MSTR}
    translation {y=0.5}
   }
   n7 = "Transform (Modify) Bone - Bone: dyn_rear_cloth_extra_l_MSTR"(ComponentPose=@n6, Alpha=1, Translation="0,-0.7,0", Rotation="0,0,0", Scale="1,1,1") {
    rotationMode=BMM_Additive
    rotationSpace=BCS_BoneSpace
    translationMode=BMM_Additive
    translationSpace=BCS_BoneSpace
    alphaBoolBlend {blendOption=Linear}
    boneToModify {boneName=dyn_rear_cloth_extra_l_MSTR}
    translation {y=-0.7}
   }
   n8 = "Transform (Modify) Bone - Bone: dyn_rear_cloth_extra_r_MSTR"(ComponentPose=@n7, Alpha=1, Translation="0,0.7,0", Rotation="0,0,0", Scale="1,1,1") {
    rotationMode=BMM_Additive
    rotationSpace=BCS_BoneSpace
    scaleMode=BMM_Replace
    translationMode=BMM_Additive
    translationSpace=BCS_BoneSpace
    alphaBoolBlend {blendOption=Linear}
    boneToModify {boneName=dyn_rear_cloth_extra_r_MSTR}
    translation {y=0.7}
   }
   Save cached pose 'Post RBAN'(Pose=Component To Local(ComponentPose=@n8)) {cacheName=Post RBAN}
   n9 = "Transform (Modify) Bone - Bone: dyn_rear_cloth_c_MSTR"!(
    ComponentPose=Local To Component(LocalPose=Use cached pose 'Post Deformation Rig'())
    bAlphaBoolEnabled:bool=True
    Translation="-1,0,0"
    Rotation="0,-100,0"
    Scale="1,1,1"
   ) -> (Pose) {
    alphaInputType=Bool
    rotationMode=BMM_Additive
    rotationSpace=BCS_BoneSpace
    translationMode=BMM_Additive
    translationSpace=BCS_BoneSpace
    alphaBoolBlend {blendInTime=0.1, blendOption=Linear, blendOutTime=0.1}
    boneToModify {boneName=dyn_rear_cloth_c_MSTR}
    rotation {yaw=-100}
    translation {x=-1}
   }
   n10 = "Transform (Modify) Bone - Bone: dyn_rear_cloth_l_MSTR"!(ComponentPose=@n9, bAlphaBoolEnabled:bool=True, Translation="-1,0,0", Rotation="0,-100,0", Scale="1,1,1") -> (Pose) {
    alphaInputType=Bool
    rotationMode=BMM_Additive
    rotationSpace=BCS_BoneSpace
    translationMode=BMM_Additive
    translationSpace=BCS_BoneSpace
    alphaBoolBlend {blendInTime=0.1, blendOption=Linear, blendOutTime=0.1}
    boneToModify {boneName=dyn_rear_cloth_l_MSTR}
    rotation {yaw=-100}
    translation {x=-1}
   }
   n11 = "Transform (Modify) Bone - Bone: dyn_rear_cloth_r_MSTR"!(ComponentPose=@n10, bAlphaBoolEnabled:bool=True, Translation="1,0,0", Rotation="0,-100,0", Scale="1,1,1") -> (Pose) {
    alphaInputType=Bool
    rotationMode=BMM_Additive
    rotationSpace=BCS_BoneSpace
    translationMode=BMM_Additive
    translationSpace=BCS_BoneSpace
    alphaBoolBlend {blendInTime=0.1, blendOption=Linear, blendOutTime=0.1}
    boneToModify {boneName=dyn_rear_cloth_r_MSTR}
    rotation {yaw=-100}
    translation {x=1}
   }
   n12 = "Transform (Modify) Bone - Bone: dyn_floor_collision"!(ComponentPose=@n11, bAlphaBoolEnabled:bool=True, Translation="0,0,55", Rotation="0,0,0", Scale="1,1,1") -> (Pose) {
    alphaInputType=Bool
    rotationMode=BMM_Additive
    rotationSpace=BCS_BoneSpace
    translationMode=BMM_Additive
    translationSpace=BCS_BoneSpace
    alphaBoolBlend {blendInTime=0.05, blendOption=Linear, blendOutTime=0.05}
    boneToModify {boneName=dyn_floor_collision}
    translation {z=55}
   }
   n13 = "Transform (Modify) Bone - Bone: dyn_rear_cloth_c_MSTR"(ComponentPose=@n12, bAlphaBoolEnabled=True, Translation="-1,0,0", Rotation="0,-100,0", Scale="1,1,1") {
    alphaInputType=Bool
    rotationMode=BMM_Additive
    rotationSpace=BCS_BoneSpace
    translationMode=BMM_Additive
    translationSpace=BCS_BoneSpace
    alphaBoolBlend {blendInTime=0.1, blendOption=Linear, blendOutTime=0.1}
    boneToModify {boneName=dyn_rear_cloth_c_MSTR}
    rotation {yaw=-100}
    translation {x=-1}
   }
   n14 = "Transform (Modify) Bone - Bone: dyn_rear_cloth_l_MSTR"(ComponentPose=@n13, bAlphaBoolEnabled=True, Translation="-1,0,0", Rotation="0,-100,0", Scale="1,1,1") {
    alphaInputType=Bool
    rotationMode=BMM_Additive
    rotationSpace=BCS_BoneSpace
    translationMode=BMM_Additive
    translationSpace=BCS_BoneSpace
    alphaBoolBlend {blendInTime=0.1, blendOption=Linear, blendOutTime=0.1}
    boneToModify {boneName=dyn_rear_cloth_l_MSTR}
    rotation {yaw=-100}
    translation {x=-1}
   }
   n15 = "Transform (Modify) Bone - Bone: dyn_rear_cloth_r_MSTR"(ComponentPose=@n14, bAlphaBoolEnabled=True, Translation="1,0,0", Rotation="0,-100,0", Scale="1,1,1") {
    alphaInputType=Bool
    rotationMode=BMM_Additive
    rotationSpace=BCS_BoneSpace
    translationMode=BMM_Additive
    translationSpace=BCS_BoneSpace
    alphaBoolBlend {blendInTime=0.1, blendOption=Linear, blendOutTime=0.1}
    boneToModify {boneName=dyn_rear_cloth_r_MSTR}
    rotation {yaw=-100}
    translation {x=1}
   }
   Save cached pose 'Post Pose'(Pose=Component To Local(ComponentPose=@n15)) {cacheName=Post Pose}
   n16 = Anim Dynamics - Bone: dyn_shoe_tag_l!(
    SimulationSpace:AnimPhysSimSpaceType Enum
    ComponentPose=Local To Component(LocalPose=Use cached pose 'Post RBAN'())
    Alpha=1
    ExternalForce="0,0,0"
   ) -> (Pose) {
    angularDampingOverride=0.99
    angularSpringConstant=10000
    bAngularSpring=true
    bOverrideAngularDamping=true
    bUseGravityOverride=true
    gravityScale=0.35
    simulationSpace=RootRelative
    alphaBoolBlend {blendOption=Linear}
    boundBone {boneName=dyn_shoe_tag_l}
    componentAppliedLinearAccClamp=<1000 1000 1000>
    componentLinearAccScale=<1 1 1>
    componentLinearVelScale=<1 1 1>
    gravityOverride {z=-980}
    physicsBodyDefinitions [{
      boundBone {boneName=dyn_shoe_tag_l}
      boxExtents=<20 25 20>
      constraintSetup {angularLimitsMax=<15 15 30>, angularLimitsMin=<-15 -15 -30>, angularTarget=<1 0 0>}
      localJointOffset=<4 0 0>
     }]
   }
   n17 = Anim Dynamics - Bone: dyn_shoe_tag_r!(SimulationSpace:AnimPhysSimSpaceType Enum, ComponentPose=@n16, Alpha=1, ExternalForce="0,0,0") -> (Pose) {
    angularDampingOverride=0.99
    angularSpringConstant=10000
    bAngularSpring=true
    bOverrideAngularDamping=true
    bUseGravityOverride=true
    gravityScale=0.35
    simulationSpace=RootRelative
    alphaBoolBlend {blendOption=Linear}
    boundBone {boneName=dyn_shoe_tag_r}
    componentAppliedLinearAccClamp=<1000 1000 1000>
    componentLinearAccScale=<1 1 1>
    componentLinearVelScale=<1 1 1>
    gravityOverride {z=-980}
    physicsBodyDefinitions [{
      boundBone {boneName=dyn_shoe_tag_r}
      boxExtents=<20 25 20>
      constraintSetup {angularLimitsMax=<15 15 30>, angularLimitsMin=<-15 -15 -30>, angularTarget=<1 0 0>}
      localJointOffset=<-4 0 0>
     }]
   }
   n18 = Anim Dynamics - Bone: dyn_thigh_l_zipper_a!(SimulationSpace:AnimPhysSimSpaceType Enum=Component, ComponentPose=@n17, Alpha=1, ExternalForce="0,0,0") -> (Pose) {
    angularDampingOverride=0.999
    angularSpringConstant=20000
    bAngularSpring=true
    bOverrideAngularDamping=true
    bUseGravityOverride=true
    gravityScale=0.35
    simulationSpace=World
    alphaBoolBlend {blendOption=Linear}
    boundBone {boneName=dyn_thigh_l_zipper_a}
    componentAppliedLinearAccClamp=<1000 1000 1000>
    componentLinearAccScale=<1 1 1>
    componentLinearVelScale=<1 1 1>
    gravityOverride {z=-980}
    physicsBodyDefinitions [{
      boundBone {boneName=dyn_thigh_l_zipper_a}
      boxExtents=<35 40 40>
      constraintSetup {angularLimitsMax=<15 15 15>, angularLimitsMin=<-15 -15 -15>, angularTarget=<1 0 0>}
      localJointOffset=<4 0 0>
     }]
   }
   n19 = Anim Dynamics - Bone: dyn_thigh_l_zipper_b!(SimulationSpace:AnimPhysSimSpaceType Enum=World, ComponentPose=@n18, Alpha=1, ExternalForce="0,0,0") -> (Pose) {
    angularDampingOverride=0.999
    angularSpringConstant=20000
    bAngularSpring=true
    bOverrideAngularDamping=true
    bUseGravityOverride=true
    gravityScale=0.35
    simulationSpace=World
    alphaBoolBlend {blendOption=Linear}
    boundBone {boneName=dyn_thigh_l_zipper_b}
    componentAppliedLinearAccClamp=<1000 1000 1000>
    componentLinearAccScale=<1 1 1>
    componentLinearVelScale=<1 1 1>
    gravityOverride {z=-980}
    physicsBodyDefinitions [{
      boundBone {boneName=dyn_thigh_l_zipper_b}
      boxExtents=<35 40 40>
      constraintSetup {angularLimitsMax=<15 15 15>, angularLimitsMin=<-15 -15 -15>, angularTarget=<1 0 0>}
      localJointOffset=<4 0 0>
     }]
   }
   n20 = Anim Dynamics - Bone: dyn_thigh_r_zipper_a!(SimulationSpace:AnimPhysSimSpaceType Enum=World, ComponentPose=@n19, Alpha=1, ExternalForce="0,0,0") -> (Pose) {
    angularDampingOverride=0.999
    angularSpringConstant=20000
    bAngularSpring=true
    bOverrideAngularDamping=true
    bUseGravityOverride=true
    gravityScale=0.35
    simulationSpace=World
    alphaBoolBlend {blendOption=Linear}
    boundBone {boneName=dyn_thigh_r_zipper_a}
    componentAppliedLinearAccClamp=<1000 1000 1000>
    componentLinearAccScale=<1 1 1>
    componentLinearVelScale=<1 1 1>
    gravityOverride {z=-980}
    physicsBodyDefinitions [{
      boundBone {boneName=dyn_thigh_r_zipper_a}
      boxExtents=<35 40 40>
      constraintSetup {angularLimitsMax=<15 15 15>, angularLimitsMin=<-15 -15 -15>, angularTarget=<1 0 0>}
      localJointOffset=<-4 0 0>
     }]
   }
   n21 = Anim Dynamics - Bone: dyn_thigh_r_zipper_b!(SimulationSpace:AnimPhysSimSpaceType Enum=World, ComponentPose=@n20, Alpha=1, ExternalForce="0,0,0") -> (Pose) {
    angularDampingOverride=0.999
    angularSpringConstant=20000
    bAngularSpring=true
    bOverrideAngularDamping=true
    bUseGravityOverride=true
    gravityScale=0.35
    simulationSpace=World
    alphaBoolBlend {blendOption=Linear}
    boundBone {boneName=dyn_thigh_r_zipper_b}
    componentAppliedLinearAccClamp=<1000 1000 1000>
    componentLinearAccScale=<1 1 1>
    componentLinearVelScale=<1 1 1>
    gravityOverride {z=-980}
    physicsBodyDefinitions [{
      boundBone {boneName=dyn_thigh_r_zipper_b}
      boxExtents=<35 40 40>
      constraintSetup {angularLimitsMax=<15 15 15>, angularLimitsMin=<-15 -15 -15>, angularTarget=<1 0 0>}
      localJointOffset=<-4 0 0>
     }]
   }
   Save cached pose 'Post Extra Dynamics'(Pose=Component To Local(ComponentPose=@n21)) {cacheName=Post Extra Dynamics}
  ]}
 EventGraph {nodes [
   Event BlueprintUpdateAnimation()
   > Branch(Condition=Get bRigidBodyEnabled())
   > n1 = Gravity Override State Function DEPRECATED(
    CharacterPartAnimInstance=Self-Reference()
    GameTimeInSeconds=GetGameTimeInSeconds()
    GravityOverrideParams=Get Gravity Override Params()
    RBAN Params=Get RBAN Params()
   )
   > Set Gravity Override(Gravity Override=@n1)
   > Set RBAN Settings Out(RBAN Settings Out:RBANSetting Struct Structure=Get RBAN Settings Default() -> (RBAN Settings Default:RBANSetting Struct Structure)) -> (Output_Get:RBANSetting Struct Structure)
   > n2 = Branch(Condition=Get bIsInFrontEnd())
   > Set RBAN Settings Out(RBAN Settings Out=Get RBAN Settings Frontend() -> (RBAN Settings Frontend:RBANSetting Struct Structure))
   > Set Simulation Space(Simulation Space:AnimPhysSimSpaceType Enum=World) -> (Output_Get:AnimPhysSimSpaceType Enum)
   Set Simulation Space(execute=@n2.else, Simulation Space=Component)
   > Branch(Condition=Get bIsSliding())
   > Set RBAN Settings Out(RBAN Settings Out=Get RBAN Settings Sliding() -> (RBAN Settings Sliding:RBANSetting Struct Structure))
  ]}
]
variables [
 Gravity Override {type=Vector, default=<0 0 -980>}
 Gravity Override Params {
  type=GravityOverrideParamsStruct
  default {
   onGround_Moving {gravityOverride=<0 -300 -700>}
   skydive_DivingDown {windAmplitude=<50 500 250>, windFrequency=<1.5 30 30>}
   skydive_FloatingDown {gravityOverride=<0 -900 500>, windAmplitude=<100 -250 200>, windFrequency=<1.5 5 8>, windOffset=<0 -300 200>}
   tactical_Sprint {gravityOverride=<0 -1200 -500>}
  }
 }
 IsMoving {type=boolean}
 RBAN Params {
  type=RBANParamsStruct
  default {
   dBNO {simSpaceSettings {
     maxAngularAcceleration=50
     maxAngularVelocity=7
     maxLinearAcceleration=200
     maxLinearVelocity=300
     velocityScaleZ=1
     externalLinearDragV=<0.2 0.2 0.2>
    }}
   dBNO_Carried {simSpaceSettings {
     maxAngularAcceleration=50
     maxAngularVelocity=7
     maxLinearAcceleration=200
     maxLinearVelocity=300
     velocityScaleZ=1
     externalLinearDragV=<0.2 0.2 0.2>
    }}
   emoteOrMelee {simSpaceSettings {
     maxAngularAcceleration=50
     maxAngularVelocity=7
     maxLinearAcceleration=200
     maxLinearVelocity=300
     velocityScaleZ=1
     externalLinearDragV=<0.2 0.2 0.2>
    }}
   front_End {simSpaceSettings {
     maxAngularAcceleration=50
     maxAngularVelocity=7
     maxLinearAcceleration=200
     maxLinearVelocity=300
     velocityScaleZ=1
     externalLinearDragV=<0.2 0.2 0.2>
    }}
   onGround_Crouch_Moving {simSpaceSettings {
     maxAngularAcceleration=50
     maxAngularVelocity=7
     maxLinearAcceleration=200
     maxLinearVelocity=300
     velocityScaleZ=1
     externalLinearDragV=<0.2 0.2 0.2>
    }}
   onGround_Falling {simSpaceSettings {
     maxAngularAcceleration=50
     maxAngularVelocity=7
     maxLinearAcceleration=200
     maxLinearVelocity=300
     velocityScaleZ=1
     externalLinearDragV=<0.2 0.2 0.2>
    }}
   onGround_Moving {simSpaceSettings {
     maxAngularAcceleration=100
     maxAngularVelocity=12
     maxLinearAcceleration=200
     maxLinearVelocity=300
     velocityScaleZ=1
     externalLinearDragV=<0.2 0.2 0.2>
    }}
   onGround_Standing {simSpaceSettings {
     maxAngularAcceleration=50
     maxAngularVelocity=7
     maxLinearAcceleration=200
     maxLinearVelocity=300
     velocityScaleZ=1
     externalLinearDragV=<0.2 0.2 0.2>
    }}
   rail_Grinding {simSpaceSettings {
     maxAngularAcceleration=1000
     maxAngularVelocity=3
     maxLinearAcceleration=300
     maxLinearVelocity=700
     velocityScaleZ=0
     worldAlpha=0.1
     externalLinearDragV=<0 0.5 0>
    }}
   riding_Moving {simSpaceSettings {
     maxAngularAcceleration=50
     maxAngularVelocity=7
     maxLinearAcceleration=200
     maxLinearVelocity=300
     velocityScaleZ=1
     externalLinearDragV=<0.2 0.2 0.2>
    }}
   riding_Moving_Targeting {simSpaceSettings {
     maxAngularAcceleration=50
     maxAngularVelocity=7
     maxLinearAcceleration=200
     maxLinearVelocity=300
     velocityScaleZ=1
     externalLinearDragV=<0.2 0.2 0.2>
    }}
   riding_Standing {simSpaceSettings {
     maxAngularAcceleration=50
     maxAngularVelocity=7
     maxLinearAcceleration=200
     maxLinearVelocity=300
     velocityScaleZ=1
     externalLinearDragV=<0.2 0.2 0.2>
    }}
   skydive_DivingDown {simSpaceSettings {
     maxAngularAcceleration=50
     maxAngularVelocity=7
     maxLinearAcceleration=200
     maxLinearVelocity=300
     velocityScaleZ=1
     externalLinearDragV=<0.2 0.2 0.2>
    }}
   skydive_FloatingDown {
    componentLinearAccScale=<0 0.5 0>
    componentLinearVelScale=<0 0.5 0>
    simSpaceSettings {
     maxAngularAcceleration=50
     maxAngularVelocity=7
     maxLinearAcceleration=200
     maxLinearVelocity=300
     velocityScaleZ=1
     externalLinearDragV=<0.2 0.2 0.2>
    }
   }
   skydive_GoingUp {simSpaceSettings {
     maxAngularAcceleration=50
     maxAngularVelocity=7
     maxLinearAcceleration=200
     maxLinearVelocity=300
     velocityScaleZ=1
     externalLinearDragV=<0.2 0.2 0.2>
    }}
   skydive_Parachuting {
    componentLinearAccScale=<0 0.1 0.3>
    componentLinearVelScale=<0 0.1 0.3>
    simSpaceSettings {
     maxAngularAcceleration=50
     maxAngularVelocity=7
     maxLinearAcceleration=200
     maxLinearVelocity=300
     velocityScaleZ=1
     externalLinearDragV=<0.2 0.2 0.2>
    }
   }
   skydive_Surfing {simSpaceSettings {
     maxAngularAcceleration=50
     maxAngularVelocity=7
     maxLinearAcceleration=200
     maxLinearVelocity=300
     velocityScaleZ=1
     externalLinearDragV=<0.2 0.2 0.2>
    }}
   swimming {simSpaceSettings {
     maxAngularAcceleration=50
     maxAngularVelocity=7
     maxLinearAcceleration=200
     maxLinearVelocity=300
     velocityScaleZ=1
     externalLinearDragV=<0.2 0.2 0.2>
    }}
   tactical_Sprint {simSpaceSettings {
     maxAngularAcceleration=100
     maxAngularVelocity=12
     maxLinearAcceleration=600
     maxLinearVelocity=600
     velocityScaleZ=1
     externalAngularVelocity=<0 0.5 0>
     externalLinearDragV=<0.2 0.2 0.2>
     externalLinearVelocity=<0 0.5 0>
    }}
   ziplining {simSpaceSettings {
     maxAngularAcceleration=50
     maxAngularVelocity=7
     maxLinearAcceleration=200
     maxLinearVelocity=300
     velocityScaleZ=1
     externalLinearDragV=<0.2 0.2 0.2>
    }}
  }
 }
 RBAN Settings Default {
  type=RBANSettingStruct
  default {
   componentLinearAccScale=<0 0 0>
   componentLinearVelScale=<0 0 0>
   simSpaceSettings {
    dampingAlpha=1
    maxAngularAcceleration=20
    maxAngularVelocity=8
    maxLinearAcceleration=200
    maxLinearVelocity=500
    velocityScaleZ=1
    worldAlpha=1
    externalAngularVelocity=<0 0 0>
    externalLinearDragV=<0.2 0.2 0.2>
    externalLinearVelocity=<0 0 0>
   }
  }
 }
 RBAN Settings Frontend {
  type=RBANSettingStruct
  default {
   componentLinearAccScale=<0 0 0>
   componentLinearVelScale=<0 0 0>
   simSpaceSettings {
    dampingAlpha=1
    maxAngularAcceleration=10
    maxAngularVelocity=4
    maxLinearAcceleration=100
    maxLinearVelocity=250
    velocityScaleZ=0.7
    worldAlpha=0.7
    externalAngularVelocity=<0 0 0>
    externalLinearDragV=<0.2 0.2 0.2>
    externalLinearVelocity=<0 0 0>
   }
  }
 }
 RBAN Settings Out {
  type=RBANSettingStruct
  default {
   componentLinearAccScale=<0 0 0>
   componentLinearVelScale=<0 0 0>
   simSpaceSettings {
    dampingAlpha=1
    maxAngularAcceleration=1000
    maxAngularVelocity=3
    maxLinearAcceleration=300
    maxLinearVelocity=700
    velocityScaleZ=0
    worldAlpha=1
    externalAngularVelocity=<0 0 0>
    externalLinearDragV=<0 0.5 0>
    externalLinearVelocity=<0 0 0>
   }
  }
 }
 RBAN Settings Sliding {
  type=RBANSettingStruct
  default {
   componentLinearAccScale=<0 0 0>
   componentLinearVelScale=<0 0 0>
   simSpaceSettings {
    dampingAlpha=1
    maxAngularAcceleration=10
    maxAngularVelocity=2
    maxLinearAcceleration=10
    maxLinearVelocity=10
    velocityScaleZ=0
    worldAlpha=1
    externalAngularVelocity=<0 0 0>
    externalLinearDragV=<0 0.5 0>
    externalLinearVelocity=<0 0 0>
   }
  }
 }
 Simulation Space {type=AnimPhysSimSpaceType, default=World}
]
